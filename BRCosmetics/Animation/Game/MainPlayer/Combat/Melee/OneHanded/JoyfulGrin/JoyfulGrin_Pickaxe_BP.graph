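# Animation Blueprint for the JoyfulGrin pickaxe weapon skeleton, choosing between a front-end pose and an idle pose.
# A single-state machine uses the front-end pose when bIsInFrontEnd is true and bIsOnPawn is false, otherwise the idle pose.
# The result passes through local/component space conversions into DefaultSlot for montages.
# The handle bone transform node is not wired to the output, and the event graph contains only unconnected nodes.
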AnimBlueprint /BRCosmetics/Animation/Game/MainPlayer/Combat/Melee/OneHanded/JoyfulGrin/JoyfulGrin_Pickaxe_BP
target_skeleton=/Game/Weapons/FORT_Melee/Pickaxe_JoyfulGrin/Meshes/JoyfulGrin_Axe_Skeleton
graphs [
 AnimGraph {nodes [
   Local Space Ref Pose()
   "Transform (Modify) Bone - Bone: Melee_TwoHanded_Handle"(
    ComponentPose="Constraint - Melee_TwoHanded_Handle ()"!(ComponentPose, Alpha=1) -> (Pose) {alphaBoolBlend {blendOption=Linear}, boneToModify {boneName=Melee_TwoHanded_Handle}}
    Alpha=1
    Translation="0,0,0"
    Rotation="0,-90,5"
    Scale="1,1,1"
   ) {rotationMode=BMM_Replace, alphaBoolBlend {blendOption=Linear}, boneToModify {boneName=Melee_TwoHanded_Handle}, rotation {roll=5, yaw=-90}}
   AnimGraphNode_StateMachine_0 = Idle_State() {is=title AnimGraphNode_StateMachine}
   n1 = Local To Component(LocalPose=@AnimGraphNode_StateMachine_0)
   Output Pose(Result=Slot 'DefaultSlot'(Source=Component To Local(ComponentPose=@n1))) {layerGroup=None}
  ]}
 Idle_State {entry=Idle, path=AnimGraph.AnimGraphNode_StateMachine_0.Idle_State, ids=id, nodes [AnimStateNode_0 = Idle()]}
 Idle {
  path=AnimGraph.AnimGraphNode_StateMachine_0.Idle_State.AnimStateNode_0.Idle
  nodes [
   n1 = NOT Boolean(A=Get bIsOnPawn() -> (bIsOnPawn:bool))
   n2 = Get bIsInFrontEnd()
   Output Animation Pose(Result=Blend Poses by bool(
     BlendPose_0="Sequence Evaluator (by time) 'JoyfulGrin_FE_Pose_WEP'"!(ExplicitTime:float=0) -> (Pose:Pose Link Structure) {
      is=title AnimGraphNode_SequenceEvaluator
      sequence=&/JoyfulGrin_FE_Pose_WEP
     }
     BlendPose_1="Sequence Evaluator (by time) 'JoyfulGrin_Idle_Pose_WEP'"!(ExplicitTime:float=0) -> (Pose:Pose Link Structure) {
      is=title AnimGraphNode_SequenceEvaluator
      sequence=&/JoyfulGrin_Idle_Pose_WEP
     }
     bActiveValue=AND Boolean(A=@n2, B=@n1)
    ))
  ]
 }
 EventGraph {nodes [
   TryGetPawnOwner()
   Event BlueprintUpdateAnimation()
  ]}
]
variables []
